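# Unlit translucent, two-sided Niagara material for the Safari Gnome pickaxe idle effect, drawing animated polka dots.
# Panning noise blends two colors, tinted by particle color through an emissive-adjustment function.
# A tiled Frac and SphereMask form dots, with radius driven by noise and a sine edge falloff.
# Opacity math adds noise and view-angle fade; dot tiling uses dynamic DotTileU on both axes, leaving DotTileV and scalar tile parameters unused.
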Material /BRCosmetics/Weapons/FORT_Melee/Pickaxe_SafariGnome/FX/M_Pickaxe_SafariGnome_Idle_Dot
graph {
 properties {
  bUsedWithNiagaraSprites=true
  blendMode=BLEND_Translucent
  pixelDepthOffsetMode=PDOM_Legacy
  shadingModel=MSM_Unlit
  twoSided=true
  usedShadingModels=Unlit
 }
 nodes [
  n1 = ParticleColor()
  n2 = VectorParameter() {parameterName=Color_1, defaultValue=<rgba 0 0.2232 0.4427 1>}
  n3 = VectorParameter() {parameterName=Color_2, defaultValue=<rgba 0.3581 0.0064 0.4115 1>}
  n4 = TextureSample(UVs=Panner(Coordinate=TextureCoordinate() {uTiling=2}) {speedX=0.1, speedY=0.3}) {texture=/Game/Effects/Fort_Effects/Emotes/UltraLight/Texture/T_SoftNoise_BC7}
  n5 = Saturate(None=Multiply(A=Power(Base=@n4.G) {constExponent=1.3}) {constB=2})
  MaterialFunctionCall(Emissive=Multiply(A=@n1.RGB, B=LinearInterpolate(A=@n2.RGB, B=@n3.RGB, Alpha=@n5))) {materialFunction=&MF_FX_EmissiveAdjustments}
  n6 = TextureCoordinate()
  n7 = DynamicParameter() {defaultValue=<rgba 10 10 1 1>, paramNames [DotTileU, DotTileV, DotTileMaster, Param4]}
  n8 = Multiply(A=@n7.DotTileMaster, B=AppendVector(A=@n7.DotTileU, B=@n7.DotTileU))
  n9 = TextureCoordinate()
  n10 = Multiply(A=Sine(None=ComponentMask(None=@n9) {r=true}) {period=2}, B=Sine(None=ComponentMask(None=@n9) {g=true}) {period=2})
  n11 = Multiply(A=Power(Base=Multiply(A=@n4.R, B=@n10)) {constExponent=0.7})
  n12 = SphereMask(
   A=Frac(None=Multiply(A=@n6, B=@n8) {constB=10})
   B=Constant() {r=0.5}
   Radius=Multiply(A=ScalarParameter() {defaultValue=1, parameterName=Polka Dot Size}, B=Clamp(None=@n11) {maxDefault=0.35, minDefault=0.03})
   Hardness=ScalarParameter() {defaultValue=0.1, parameterName=Polka Dot Hardness, sliderMax=1}
  )
  n13 = Abs(None=DotProduct(A=CameraVectorWS(), B=VertexNormalWS()))
  Multiply(A=Multiply(A=Multiply(A=@n1.A, B=@n12), B=@n4.B), B=Saturate(None=Power(Base=@n13) {constExponent=0.7}))
  ScalarParameter() {defaultValue=10, parameterName=Polka Dot Tile U}
  ScalarParameter() {defaultValue=60, parameterName=Polka Dot Tile V}
 ]
}
