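# Niagara swing trail for the SafariGnome pickaxe, drawn by one animation-trail emitter with two ribbon renderers.
# Particle update blends between a dark grey look and a magenta-to-light-blue look, driven by System.TODReactivity when System.Reactive is set, otherwise by User.Variant.
# The same blend sets alpha, a clean-trail color and two dynamic parameter vectors.
# Colors are tinted by time-of-day color, and SwingGeneralFade scales opacity.
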NiagaraSystem /BRCosmetics/Weapons/FORT_Melee/Pickaxe_SafariGnome/FX/NS_Pickaxe_SafariGnome_Trail
dependencies {
 usedDataInterfaces [&NiagaraDataInterfaceColorCurve, &NiagaraDataInterfaceCurve, &NiagaraDataInterfaceSocketReader]
 usedDynamicInputs [&FNTimeOfDayColor, &SetBoolByFloatComparison]
 usedModules [
  &self:SystemSpawnScript.SystemScriptSource.SystemScriptGraph.NiagaraNodeAssignment_1.SetVariables_#1
  &self:SystemSpawnScript.SystemScriptSource.SystemScriptGraph.NiagaraNodeAssignment_0.SetVariables_#2
  &SystemState
  &self:SystemSpawnScript.SystemScriptSource.SystemScriptGraph.NiagaraNodeAssignment_2.SetVariables_#3
  &self:SystemSpawnScript.SystemScriptSource.SystemScriptGraph.NiagaraNodeAssignment_3.SetVariables_#4
  &self:SystemSpawnScript.SystemScriptSource.SystemScriptGraph.NiagaraNodeAssignment_4.SetVariables_#5
  &self:SystemSpawnScript.SystemScriptSource.SystemScriptGraph.NiagaraNodeAssignment_5.SetVariables_#6
 ]
}
emitters {Trail {
  emitterSpawnScript [&FN_RibbonIDOffsetUtility()]
  emitterUpdateScript [
   &EmitterState()
   &FN_SpawnAnimTrail()
  ]
  particleSpawnScript [
   &FN_InitializeAnimTrailBoneIndecies()
   &FN_InitializeAnimTrail(Lifetime=0.35)
  ]
  particleUpdateScript [
   &FN_UpdateAnimTrail() {module=FN_UpdateAnimTrail001}
   /SetVariables(Particles.AdjustedNormalizedAge:NiagaraFloat=&Multiply_Float(A=$Particles.NormalizedAge, B=&Divide_Float(A=$Particles.Lifetime, B=0.18))) {isSetParameters=true}
   &SolveForcesAndVelocity(Clamp Velocity=false, Limit Acceleration=false)
   &ScaleColor(
    Scale Alpha=&FloatFromCurve(
     FloatCurve=DataInterface() {propertyValues {Curve {keys [{interpMode=RCIM_Cubic, tangentMode=RCTM_User, time=0, value=1}, {interpMode=RCIM_Cubic, tangentMode=RCTM_User, time=0.7372, value=0}]}}}
     Scale Curve=&Multiply_Float(
      A=&MakeCustomFloatFromBool(
       Bool=$System.Reactive
       True Float=&Lerp_Float(A=3, B=1.5, Alpha=$System.TODReactivity)
       False Float=&Lerp_Float(A=3, B=1.5, Alpha=$User.Variant)
      )
      B=$System.SwingGeneralFade
     )
    )
    Scale RGB=&MakeVectorFromLinearColorRGB(Color=&MultiplyLinearColor(
      Linear Color=&MakeCustomLinearColorFromBool(
       Bool=$System.Reactive
       True Color=&LerpLinearColors(
        StartColor=&ColorFromCurve(
         ColorCurve=DataInterface() {propertyValues {
           AlphaCurve {keys [<at 0.5226 1 RCIM_Cubic>,<at 1 0 RCIM_Cubic>]}
           BlueCurve {keys [<at 0 0.1146 RCIM_Cubic>]}
           GreenCurve {keys [<at 0 0.1146 RCIM_Cubic>]}
           RedCurve {keys [<at 0 0.1146 RCIM_Cubic>]}
          }}
         CurveIndex=$Particles.AdjustedNormalizedAge
        )
        EndColor=&ColorFromCurve(
         ColorCurve=DataInterface() {propertyValues {
           AlphaCurve {keys [<at 0 1 RCIM_Cubic>,<at 1 0 RCIM_Cubic>]}
           BlueCurve {keys [<at 0.1607 2 RCIM_Cubic>,<at 0.57 1>]}
           GreenCurve {keys [<at 0.1607 0 RCIM_Cubic>,<at 0.57 0.7653>]}
           RedCurve {keys [<at 0.1607 1.4686 RCIM_Cubic>,<at 0.57 0.3698>]}
          }}
         CurveIndex=$Particles.AdjustedNormalizedAge
        )
        LerpFactor=$System.TODReactivity
       )
       False Color=&LerpLinearColors(
        StartColor=&ColorFromCurve(
         ColorCurve=DataInterface() {propertyValues {
           AlphaCurve {keys [<at 0.5226 1 RCIM_Cubic>,<at 1 0 RCIM_Cubic>]}
           BlueCurve {keys [<at 0 0.1146 RCIM_Cubic>]}
           GreenCurve {keys [<at 0 0.1146 RCIM_Cubic>]}
           RedCurve {keys [<at 0 0.1146 RCIM_Cubic>]}
          }}
         CurveIndex=$Particles.AdjustedNormalizedAge
        )
        EndColor=&ColorFromCurve(
         ColorCurve=DataInterface() {propertyValues {
           AlphaCurve {keys [<at 0 1 RCIM_Cubic>,<at 1 0 RCIM_Cubic>]}
           BlueCurve {keys [<at 0.1607 2 RCIM_Cubic>,<at 0.57 1>]}
           GreenCurve {keys [<at 0.1607 0 RCIM_Cubic>,<at 0.57 0.7653>]}
           RedCurve {keys [<at 0.1607 1.4686 RCIM_Cubic>,<at 0.57 0.3698>]}
          }}
         CurveIndex=$Particles.AdjustedNormalizedAge
        )
        LerpFactor=$User.Variant
       )
      )
      Scale Factor=&MakeVector4_FromLinearColor(LinearColor=$System.ToDColor)
     ))
    ScaleA=true
    ScaleRGB=true
   )
   /SetVariables(Particles.CleanTrailColor:LinearColor=&MakeLinearColorFromVectorAndFloat(
     "Vector (RGB)"=&Multiply_VectorByFloat(
      Vector=&MakeVectorFromLinearColorRGB(Color=$System.ToDColor)
      Float=&MakeCustomFloatFromBool(
       Bool=$System.Reactive
       True Float=&Lerp_Float(A=0.17, B=1, Alpha=$System.TODReactivity)
       False Float=&Lerp_Float(A=0.17, B=1, Alpha=$User.Variant)
      )
     )
     "Float (Alpha)"=&FloatFromCurve(
      FloatCurve=DataInterface() {propertyValues {Curve {
         defaultValue=1
         keys [
          {arriveTangent=-0.0405, interpMode=RCIM_Cubic, leaveTangent=-0.2898, tangentMode=RCTM_Break, time=0, value=1}
          {arriveTangent=-0.6838, interpMode=RCIM_Cubic, leaveTangent=-0.6838, tangentMode=RCTM_User, time=1, value=0}
         ]
        }}}
      Scale Curve=&Multiply_Float(
       A=&MakeCustomFloatFromBool(
        Bool=$System.Reactive
        True Float=&Lerp_Float(A=1.1, B=0.55, Alpha=$System.TODReactivity)
        False Float=&Lerp_Float(A=1.1, B=0.55, Alpha=$User.Variant)
       )
       B=$System.SwingGeneralFade
      )
     )
    )) {isSetParameters=true}
   /SetVariables(Particles.StateTrailDynamic:Vector4f=&MakeVector4(
     X=&MakeCustomFloatFromBool(
      Bool=$System.Reactive
      True Float=&Lerp_Float(A=0, B=0.6, Alpha=$System.TODReactivity)
      False Float=&Lerp_Float(A=0, B=0.6, Alpha=$User.Variant)
     )
     Y=0
     Z=&FloatFromCurve(
      FloatCurve=DataInterface() {propertyValues {Curve {
         defaultValue=0
         keys [{interpMode=RCIM_Cubic, tangentMode=RCTM_User, time=0, value=0}, {interpMode=RCIM_Cubic, tangentMode=RCTM_User, time=1, value=0.5}]
        }}}
      CurveIndex=$Particles.AdjustedNormalizedAge
     )
     W=0
    )) {isSetParameters=true}
   /SetVariables(Particles.StateTrailSynamic1:Vector4f=&MakeVector4(
     X=&MakeCustomFloatFromBool(
      Bool=$System.Reactive
      True Float=&Lerp_Float(A=1, B=0, Alpha=$System.TODReactivity)
      False Float=&Lerp_Float(A=1, B=0, Alpha=$User.Variant)
     )
     Y=0
     Z=0
     W=0
    )) {isSetParameters=true}
  ]
  rendererClasses [&NiagaraRibbonRendererProperties]
  renderers [
   {rendererClass=&NiagaraRibbonRendererProperties}
   {rendererClass=&NiagaraRibbonRendererProperties, rendererIndex=1}
  ]
 }}
summary {
 systemName=NS_Pickaxe_SafariGnome_Trail
 emitters [{emitterName=Trail, rendererClasses [&NiagaraRibbonRendererProperties]}]
 userVariables [
  User.bIsFrontEnd {type=NiagaraBool}
  User.bIsFrontEndPreview {type=NiagaraBool}
  User.FrontEndScale {type=NiagaraFloat, defaultValue=0.6, description=""}
  User.PhysicalSurface {type=NiagaraInt32}
  User.Reactivity {type=NiagaraFloat, defaultValue=0, description=""}
  User.RibbonIDOffset {type=NiagaraInt32}
  User.StateBlend {type=NiagaraFloat, defaultValue=0.6, description=""}
  User.StateBlendOpacity {type=NiagaraFloat, defaultValue=0, description=""}
  User.Variant {type=NiagaraFloat, defaultValue=1, description=""}
 ]
}
