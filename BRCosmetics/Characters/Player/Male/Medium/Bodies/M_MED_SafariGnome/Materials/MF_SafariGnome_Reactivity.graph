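# Material function for the SafariGnome cosmetic that blends the incoming material toward an alternate night look, behind a static switch.
# The alternate look layers TOD textures through silk, thin-film, metal LUT, clear coat and anisotropic functions, plus masked shading-model and pickaxe-emissive blends.
# Emissive optionally adds two-color directional fresnel and noise-broken halftone dots on camera-aligned virtual-plane UVs.
# Blend alpha follows time of day, a timestamp-triggered simulated cycle or a force-night parameter.
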MaterialFunction /BRCosmetics/Characters/Player/Male/Medium/Bodies/M_MED_SafariGnome/Materials/MF_SafariGnome_Reactivity
graph {nodes [
  n1 = NamedRerouteUsage()
  n2 = SetMaterialAttributes!(
   MaterialAttributes=NamedRerouteUsage()
   BaseColor=NamedRerouteUsage()
   Metallic=NamedRerouteUsage()
   Specular=NamedRerouteUsage()
   Roughness=NamedRerouteUsage()
   EmissiveColor=NamedRerouteUsage()
   Normal=NamedRerouteUsage()
  ) {attributeSetTypes [
    #1
    #2
    #3
    #4
    #5
    #6
   ]}
  n3 = MaterialFunctionCall(Float3=NamedRerouteUsage()) {materialFunction=&BreakOutFloat3Components}
  n4 = Multiply(A=ComponentMask(None=NamedRerouteUsage()) {b=true}, B=ScalarParameter() {parameterName=ShortsThinFilmInfluence})
  n5 = MaterialFunctionCall!(
   Material Attributes=MaterialFunctionCall!(Material=@n2, Alpha=NamedRerouteUsage()) {materialFunction=&/MF_Silk_TechArt_SafariGnome}.Result
   Fresnel=NamedRerouteUsage()
   Alpha=MaterialFunctionCall(X=@n3.R, Y=@n3.R, Z=Add(A=@n3.B, B=@n4)) {materialFunction=&MakeFloat3}.Result
  ) {materialFunction=&/MF_ThinFilm_TechArt_SafariGnome}
  n6 = NamedRerouteUsage()
  n7 = MaterialFunctionCall!(In=MaterialFunctionCall!(
    Material=MaterialFunctionCall!(Materials=@n5.Result, Alpha=@n6) {materialFunction=&/MF_MetalLUT_TechArt_SafariGnome}.Result
    Alpha=NamedRerouteUsage()
   ) {materialFunction=&/MF_ClearCoat_TechArt_SafariGnome}.Result) {materialFunction=&/MF_Anisotropic_TechArt_SafariGnome}
  n8 = BlendMaterialAttributes(
   A=@n7.Result
   B=SetMaterialAttributes!(MaterialAttributes=@n7.Result, SubsurfaceColor=Constant(), ShadingModel=ShadingModel()) {attributeSetTypes [#7, #8]}
   Alpha=ComponentMask(None=NamedRerouteUsage()) {r=true}
  )
  n9 = NamedRerouteUsage()
  n10 = Step(Y=Subtract(A=ComponentMask(None=@n9) {b=true}) {constB=16.5}) {constX=0}
  n11 = BlendMaterialAttributes(
   A=@n8
   B=SetMaterialAttributes!(MaterialAttributes=@n8, EmissiveColor=NamedRerouteUsage()) {attributeSetTypes [#5]}
   Alpha=Multiply(A=ScalarParameter() {parameterName=PickaxeExtraEmissive}, B=OneMinus(None=@n10))
  )
  n12 = NamedRerouteUsage()
  FunctionOutput(None=StaticSwitchParameter(True=BlendMaterialAttributes(A=@n1, B=@n11, Alpha=@n12), False=NamedRerouteUsage()) {group=TOD Reactivity, parameterName=UseTODReactivity, sortPriority=0})
  n13 = TextureSampleParameter2D() {
   group=TOD Reactivity
   parameterName=TOD_BaseColor
   samplerSource=SSM_Wrap_WorldGroupSettings
   sortPriority=1
   texture=&T_BlacksRGB
  }
  NamedRerouteDeclaration(None=@n13.RGB) {name=Texture_BaseColor, nodeColor=<rgba 1 0 0.3047 1>}
  n14 = TextureSampleParameter2D() {
   group=TOD Reactivity
   parameterName=TOD_Emissive
   samplerSource=SSM_Wrap_WorldGroupSettings
   sortPriority=2
   texture=&T_BlacksRGB
  }
  NamedRerouteDeclaration(None=@n14.RGB) {name=Texture_Emissive, nodeColor=<rgba 0 0.625 1 1>}
  n15 = Multiply(
   A=TextureCoordinate()
   B=ComponentMask(None=VectorParameter() {group=TOD HalfTone, parameterName=NoiseTiling, defaultValue=<rgba 2 1 0 1>}.RGB) {g=true, r=true}
  )
  n16 = ComponentMask(None=VectorParameter() {group=TOD HalfTone, parameterName=NoiseSpeed, defaultValue=<rgba 0.1 0.3 0 1>}.RGB) {g=true, r=true}
  NamedRerouteDeclaration(None=TextureSampleParameter2D(UVs=Panner(Coordinate=@n15, Speed=@n16) {speedX=0.1, speedY=0.3}) {group=TOD HalfTone, parameterName=Noise, texture=/Game/Effects/Fort_Effects/Emotes/UltraLight/Texture/T_SoftNoise_BC7}.RGB) {name=NoiseGradient, nodeColor=<rgba 0.5703 1 0 1>}
  n17 = TextureSampleParameter2D() {
   group=TOD HalfTone
   parameterName=TOD_TechArtMask
   samplerSource=SSM_Wrap_WorldGroupSettings
   samplerType=SAMPLERTYPE_Masks
   sortPriority=1
   texture=&T_BlackMask
  }
  NamedRerouteDeclaration(None=@n17.RGB) {name=TechArtMask, nodeColor=<rgba 1 0.9375 0 1>}
  NamedRerouteDeclaration(None=FunctionInput() {inputName=TechArtMask, sortPriority=3}) {name=IN_TechArtMask, nodeColor=<rgba 1 0 0.9141 1>}
  NamedRerouteDeclaration(None=FunctionInput() {inputName=M, sortPriority=1}) {name=IN_GMap, nodeColor=<rgba 0 1 0.7891 1>}
  n18 = VectorParameter() {group=TOD Colored Fresnel, parameterName=ColorDirection, sortPriority=3, defaultValue=<rgba 1 0 0 1>}
  n19 = DotProduct(A=@n18.RGB, B=Transform(None=VertexNormalWS()) {transformSourceType=TRANSFORMSOURCE_World, transformType=TRANSFORM_View})
  n20 = Saturate(None=@n19)
  n21 = Saturate(None=Multiply(A=@n19) {constB=-1})
  NamedRerouteDeclaration(None=Max(
    A=NamedRerouteDeclaration(None=@n20) {name=ColorAMask, nodeColor=<rgba 0.2891 1 0 1>}
    B=NamedRerouteDeclaration(None=@n21) {name=ColorBMask, nodeColor=<rgba 0.0312 0 1 1>}
   )) {name=CombinedMask, nodeColor=<rgba 0 0.5312 1 1>}
  n22 = OneMinus(None=NamedRerouteUsage())
  n23 = SmoothStep(Min=@n22, Value=ScalarParameter() {defaultValue=0.7, group=TOD Colored Fresnel, parameterName=SoftFresnelBias, sortPriority=4}) {constMax=1.1, constValue=0.6}
  n24 = Multiply(
   A=SmoothStep(Min=@n22, Value=ScalarParameter() {defaultValue=0.5, group=TOD Colored Fresnel, parameterName=SharpFresnelBias, sortPriority=5}) {constValue=0.4}
   B=ScalarParameter() {defaultValue=6, group=TOD Colored Fresnel, parameterName=SharpFresnelHardness, sortPriority=6}
  ) {constB=6}
  NamedRerouteDeclaration(None=Max(A=Saturate(None=@n23), B=Saturate(None=@n24))) {name=DoubleFresnelMask, nodeColor=<rgba 1 0 0.0234 1>}
  n25 = VectorParameter() {group=TOD Colored Fresnel, parameterName=ColorA, sortPriority=1, defaultValue=<rgba 0.1651 0.5457 0.9473 1>}
  n26 = NamedRerouteUsage()
  n27 = NamedRerouteUsage()
  n28 = VectorParameter() {group=TOD Colored Fresnel, parameterName=ColorB, sortPriority=1, defaultValue=<rgba 0.5333 0.1195 0.8796 1>}
  n29 = NamedRerouteUsage()
  n30 = NamedRerouteUsage()
  n31 = Max(A=Multiply(A=@n25.RGB, B=Multiply(A=@n26, B=@n27)), B=Multiply(A=@n28.RGB, B=Multiply(A=@n29, B=@n30)))
  n32 = ScalarParameter() {defaultValue=1, group=TOD Colored Fresnel, parameterName=FresnelEmissiveIntensity, sortPriority=2}
  NamedRerouteDeclaration(None=AppendVector(A=Multiply(A=@n31, B=@n32), B=Constant() {r=1})) {name=ColoredFresnelEmissive, nodeColor=<rgba 1 0.6094 0 1>}
  n33 = ComponentMask(None=NamedRerouteUsage()) {b=true}
  n34 = Constant() {r=90}
  n35 = Saturate(None=Divide(A=Subtract(A=@n33, B=@n34), B=Constant() {r=90}) {constB=90})
  NamedRerouteDeclaration(None=Power(Base=OneMinus(None=@n35))) {name=ZGradient, nodeColor=<rgba 0 1 0.8828 1>}
  n36 = TextureSampleParameter2D() {
   group=TOD Reactivity
   parameterName=TOD_SRM
   samplerSource=SSM_Wrap_WorldGroupSettings
   samplerType=SAMPLERTYPE_Masks
   sortPriority=3
   texture=&T_BlackMask
  }
  NamedRerouteDeclaration(None=@n36.R) {name=Texture_Specular, nodeColor=<rgba 0.9453 1 0 1>}
  n37 = MaterialFunctionCall() {materialFunction=&CameraDirectionVector}
  n38 = Multiply(A=Constant3Vector() {constant=<rgba 1 1 0 0>}, B=@n37.Result)
  n39 = Add(A=Multiply(B=Normalize(VectorInput=@n38)) {constA=0.25}, B=@n37.Result)
  n40 = Multiply(A=@n37.Result) {constB=50}
  n41 = Add(A=ActorPositionWS(), B=Constant3Vector() {constant=<rgba 0 0 90 1>})
  n42 = Constant() {r=64}
  n43 = Constant() {r=128}
  n44 = InverseLinearInterpolate(
   A=Constant() {r=45}
   B=Constant() {r=90}
   Value=MaterialFunctionCall() {materialFunction=/Engine/Functions/Engine_MaterialFunctions02/Utility/FOV}.FOV angle in degrees
  )
  NamedRerouteDeclaration(None=MaterialFunctionCall(Plane Normal Axis=Normalize(VectorInput=@n39), Plane Center=Add(A=@n40, B=@n41), UV Scale=LinearInterpolate(A=@n42, B=@n43, Alpha=@n44)) {materialFunction=&VirtualPlaneCoordinates}.UVs) {name=ObjectAlighedScreenPosition, nodeColor=<rgba 1 0.867 0 1>}
  NamedRerouteDeclaration(None=FunctionInput() {inputName=PreSkinnedLocalPosition, sortPriority=1}) {name=PreSkinnedLocalPosition, nodeColor=<rgba 0 1 0.7891 1>}
  n45 = TextureSampleParameter2D() {
   group=TOD Reactivity
   parameterName=TOD_Normals
   samplerSource=SSM_Wrap_WorldGroupSettings
   samplerType=SAMPLERTYPE_Normal
   sortPriority=4
   texture=&FlatNormalBlue
  }
  NamedRerouteDeclaration(None=@n45.RGB) {name=Texture_Normals, nodeColor=<rgba 0.9219 0 1 1>}
  n46 = Subtract(A=Distance(A=ActorPositionWS(), B=CameraPositionWS())) {constB=100, desc=Minimum}
  NamedRerouteDeclaration(None=Saturate(None=Divide(A=@n46) {constB=400, desc=Maximum})) {name=DistanceBlending, nodeColor=<rgba 0.4766 1 0 1>}
  n47 = ScalarParameter() {group=TOD Reactivity, parameterName=TOD_Timestamp}
  n48 = Subtract(A=Time(), B=@n47)
  n49 = Saturate(None=@n47)
  n50 = Multiply(A=Add(A=Multiply(A=@n48, B=@n49)) {constB=-12}, B=Divide() {constB=30, desc=One loop every 30 seconds})
  n51 = ConstantBiasScale(None=Multiply(A=Sine(None=@n50)) {constB=2})
  NamedRerouteDeclaration(None=SmoothStep(Value=Saturate(None=@n51))) {name=Simulated_TOD, nodeColor=<rgba 0.5 1 0 1>}
  NamedRerouteDeclaration(None=@n36.G) {name=Texture_Roughness, nodeColor=<rgba 0.7344 0 1 1>}
  n52 = ScalarParameter() {defaultValue=150, group=TOD HalfTone, parameterName=DotsAmount, sliderMax=512, sortPriority=20}
  n53 = Multiply(A=NamedRerouteUsage(), B=LinearInterpolate(A=@n52, B=Multiply(A=@n52) {constB=0.5}, Alpha=NamedRerouteUsage()))
  n54 = Constant() {r=1}
  n55 = Add(A=Frac(None=Rotator(Coordinate=@n53, Time=@n54) {speed=0.75})) {constB=-0.5}
  n56 = Constant()
  NamedRerouteDeclaration(None=Add(A=Distance(A=@n55, B=@n56)) {constB=-0.1, desc=to get full transition}) {name=DotsGradient, nodeColor=<rgba 1 0.0938 0 1>}
  n57 = Multiply(
   A=Add(A=NamedRerouteUsage(), B=ScalarParameter() {group=TOD HalfTone, parameterName=DotsOffset, sliderMax=0.5, sliderMin=-0.5, sortPriority=20}) {constB=-0.25}
   B=ScalarParameter() {defaultValue=1, group=TOD HalfTone, parameterName=DotsTransition, sliderMax=2, sortPriority=20}
  ) {constB=1.5}
  n58 = NamedRerouteUsage()
  n59 = SmoothStep(Min=Constant(), Max=Constant() {r=0.1}, Value=Saturate(None=Subtract(A=@n57, B=@n58)))
  n60 = ScalarParameter() {group=TOD HalfTone, parameterName=NoiseIntensity}
  n61 = ComponentMask(None=NamedRerouteUsage()) {g=true}
  n62 = LinearInterpolate(
   A=Constant()
   B=ScalarParameter() {defaultValue=1, group=TOD HalfTone, parameterName=DotsIntensity}
   Alpha=Subtract(A=NamedRerouteDeclaration(None=@n59) {name=HalfToneDotsMask, nodeColor=<rgba 0.4766 1 0 1>}, B=Multiply(A=@n60, B=@n61))
  )
  n63 = NamedRerouteUsage()
  n64 = NamedRerouteUsage()
  n65 = Multiply(
   A=LinearInterpolate(
    A=VectorParameter() {group=TOD HalfTone, parameterName=Dots Color, defaultValue=<rgba 0.1559 0.0437 0.3663 1>}.RGB
    B=VectorParameter() {group=TOD HalfTone, parameterName=Dots Stripe Color, defaultValue=<rgba 0.0931 0.4793 0.8308 1>}.RGB
    Alpha=ComponentMask(None=@n63) {r=true}
   )
   B=Multiply(A=ComponentMask(None=@n64) {g=true}) {constB=1.25}
  )
  n66 = ConstantBiasScale(None=NamedRerouteUsage()) {bias=0.5, scale=1}
  n67 = LinearInterpolate(
   A=Multiply(A=Multiply(A=@n65, B=@n66), B=NamedRerouteUsage())
   B=VectorParameter() {group=TOD HalfTone, parameterName=Dots Stripe Color, defaultValue=<rgba 0.0931 0.4793 0.8308 1>}.RGB
   Alpha=ComponentMask(None=NamedRerouteUsage()) {g=true}
  )
  NamedRerouteDeclaration(None=AppendVector(A=Multiply(A=@n62, B=@n67), B=Constant() {r=1})) {name=HalfToneDots, nodeColor=<rgba 1 0 0.6797 1>}
  NamedRerouteDeclaration(None=@n36.B) {name=Texture_Metallic, nodeColor=<rgba 0 1 0.4141 1>}
  NamedRerouteDeclaration(None=FunctionInput() {inputName=MA, inputType=FunctionInput_MaterialAttributes}) {name=IN_Material, nodeColor=<rgba 0 1 0.8828 1>}
  NamedRerouteDeclaration(None=FunctionInput() {inputName=Fresnel, inputType=FunctionInput_Scalar, sortPriority=2}) {name=IN_Fresnel, nodeColor=<rgba 0 0.4375 1 1>}
  n68 = TextureSampleParameter2D() {
   group=TOD Reactivity
   parameterName=TOD_FX_Map
   samplerSource=SSM_Wrap_WorldGroupSettings
   samplerType=SAMPLERTYPE_Masks
   sortPriority=5
   texture=&T_BlackMask
  }
  NamedRerouteDeclaration(None=@n68.RGB) {name=FXMask, nodeColor=<rgba 0.9219 0 1 1>}
  n69 = Constant()
  n70 = Constant() {r=1}
  n71 = ScalarParameter() {group=TOD Reactivity, parameterName=TOD_Timestamp}
  n72 = Multiply(
   A=LinearInterpolate(
    A=MaterialFunctionCall(Value_Morning=@n69, Value_Day=@n69, Value_Evening=@n69, Value_Night=@n70) {materialFunction=&Scalar_Time_of_Day}.Result
    B=NamedRerouteUsage()
    Alpha=Saturate(None=@n71)
   )
   B=ScalarParameter() {defaultValue=1, parameterName=EnableReactivity}
  )
  n73 = ScalarParameter() {group=TOD Reactivity, parameterName=ForceNightTime}
  NamedRerouteDeclaration(None=Saturate(None=Max(A=@n72, B=@n73))) {name=TOD_Alpha, nodeColor=<rgba 1 0.5625 0 1>}
  n74 = NamedRerouteUsage()
  n75 = Multiply(
   A=ScalarParameter() {defaultValue=1, group=TOD Reactivity, parameterName=TOD_EmissiveBrightness}
   B=ShadingPathSwitch(Default=ScalarParameter() {defaultValue=1, group=TOD Reactivity, parameterName=TOD_EmissiveMobileMultiplier}, Deferred=Constant() {r=1})
  )
  n76 = Power(
   Base=OneMinus(None=NamedRerouteUsage())
   Exp=ScalarParameter() {defaultValue=4, group=TOD Reactivity, parameterName=TOD_EmissiveFresnelPower}
  )
  n77 = AppendVector(A=Multiply(A=Multiply(A=@n74, B=@n75), B=Saturate(None=@n76)), B=Constant() {r=1})
  n78 = StaticSwitchParameter(True=Add(A=@n77, B=NamedRerouteUsage()), False=@n77) {group=TOD Colored Fresnel, parameterName=UseColoredFresnel, sortPriority=0}
  n79 = Add(A=@n78, B=NamedRerouteUsage())
  n80 = ComponentMask(None=NamedRerouteUsage()) {r=true}
  NamedRerouteDeclaration(None=StaticSwitchParameter(True=LinearInterpolate(A=@n78, B=@n79, Alpha=@n80) {constAlpha=1}, False=@n78) {group=TOD HalfTone, parameterName=UseHalfToneEffect, sortPriority=0}) {name=OUT_Emissive, nodeColor=<rgba 0.3359 0 1 1>}
 ]}
