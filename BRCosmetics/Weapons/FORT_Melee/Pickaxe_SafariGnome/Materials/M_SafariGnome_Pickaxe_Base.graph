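# Base material for the SafariGnome pickaxe, built on the 2023 character parent functions with multiple shading models, including Clear Coat and Cloth.
# PreFX attributes pass through an optional carbon fiber dual layer, toggled by the UseCarbonFiber static switch, then the SafariGnome outfit's reactivity function, then PostFX.
# With carbon fiber on, the clear coat normal output blends in an angle-corrected weave normal by the layer's mask.
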Material /BRCosmetics/Weapons/FORT_Melee/Pickaxe_SafariGnome/Materials/M_SafariGnome_Pickaxe_Base
graph {
 properties {
  bUseMaterialAttributes=true
  bUsedWithClothing=true
  bUsedWithMeshDeformer=true
  bUsedWithMorphTargets=true
  bUsedWithSkeletalMesh=true
  pixelDepthOffsetMode=PDOM_Legacy
  shadingModel=MSM_FromMaterialExpression
  subsurfaceProfile=/Game/Packages/Fortress_SharedMaterials/SubSurfaceProfile/Character_SubsurfaceProfile
  usedShadingModels=Default Lit | Subsurface | Clear Coat | Subsurface Profile | Cloth
 }
 nodes [
  n1 = MaterialFunctionCall() {materialFunction=&MF_CharacterParent_2023_PreFX}
  n2 = MaterialFunctionCall!(Alpha=NamedRerouteUsage(), In=@n1.Result) {materialFunction=/Game/Global/MaterialFunctions/TechArt/MF_CarbonFiberDualLayer}
  n3 = GetMaterialAttributes(None=@n2.Result) {attributeGetTypes [#1]}
  n4 = MaterialFunctionCall(
   BaseNormal=@n3.Normal
   AdditionalNormal=TextureSample(UVs=@n2.CarbonFiberUvs) {samplerType=SAMPLERTYPE_Normal, texture=/Game/Weapons/WeaponSkins/Wraps/Textures/T_Wrap_CarbonFiber_N}.RGB
  ) {materialFunction=&BlendAngleCorrectedNormals}
  ClearCoatNormalCustomOutput!(None=StaticSwitchParameter(True=LinearInterpolate(A=@n3.Normal, B=@n4.Result, Alpha=@n2.Mask), False=@n3.Normal) {parameterName=UseCarbonFiber})
  MaterialFunctionCall(
   Material=MaterialFunctionCall!(
    MA=StaticSwitchParameter(True=@n2.Result, False=@n1.Result) {parameterName=UseCarbonFiber}
    M=NamedRerouteUsage()
    PreSkinnedLocalPosition=NamedRerouteUsage()
    Fresnel=NamedRerouteUsage()
    TechArtMask=NamedRerouteUsage()
   ) {materialFunction=/BRCosmetics/Characters/Player/Male/Medium/Bodies/M_MED_SafariGnome/Materials/MF_SafariGnome_Reactivity}.Result
   M=@n1.M
   Pre Skinned Local Position=@n1.M
   Distance Blend=@n1.M
   Fresnel=@n1.M
  ) {materialFunction=&MF_CharacterParent_2023_PostFX}
  NamedRerouteDeclaration(None=@n1.M) {name=Mask, nodeColor=<rgba 0 1 1 1>}
  NamedRerouteDeclaration(None=@n1.Tech Art Mask) {name=TechArtMask, nodeColor=<rgba 1 0.6797 0 1>}
  NamedRerouteDeclaration(None=@n1.Fresnel) {name=Fresnel, nodeColor=<rgba 0 1 0.4375 1>}
  NamedRerouteDeclaration(None=@n1.Pre skinned Local Position) {name=PreSkinnedLocalPosition, nodeColor=<rgba 0.5469 0 1 1>}
 ]
}
