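# Physics asset for the SafariGnome two-handed pickaxe, giving collision to the handle and its dangling tassel decoration.
# The decor hangs from dyn_decor_MSTR through decor_dyn_all, which branches into a center chain ending in a tassel sphere and two side chains (ba, fr).
# Most joints allow limited swing with locked twist, while the side-chain ends lock all rotation.
# Extra constraints join both side-chain tips to each other and to the tassel, closing the chains into loops.
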PhysicsAsset /BRCosmetics/Weapons/FORT_Melee/Pickaxe_SafariGnome/Meshes/SafariGnome_Axe_PhysicsAsset_Game_And_FE
bodies {
 Melee_TwoHanded_Handle [
  {shapeType=Capsule, length=52.0795, radius=6.7863, center=<-0.3697 0.0016 27.6961>, rotation=<rot 0.0051 -90 -0.0403>}
  {shapeType=Box, extentX=1.75, extentY=4.2197, extentZ=14.04, shapeName=Melee_TwoHanded_Handle_box, center=<-0.3451 0 -6.7535>}
 ]
 decor_ba_dyn_01 [{shapeType=Capsule, length=1.021, radius=0.28, shapeName=decor_ba_dyn_01_capsule, center=<0.5105 0 0>, rotation=<rot 0 90 90>}]
 decor_ba_dyn_02 [{shapeType=Capsule, length=1.163, radius=0.28, shapeName=decor_ba_dyn_02_capsule, center=<0.5815 0 0>, rotation=<rot 0 90 90>}]
 decor_ba_dyn_03 [{shapeType=Capsule, length=1.94, radius=0.28, shapeName=decor_ba_dyn_03_capsule, center=<0.97 0 0>, rotation=<rot 0 90 90>}]
 decor_ba_dyn_04 [{shapeType=Capsule, length=1.653, radius=0.28, shapeName=decor_ba_dyn_04_capsule, center=<0.8265 0 0>, rotation=<rot 0 90 90>}]
 decor_ba_dyn_05 [{shapeType=Capsule, length=0.688, radius=0.28, shapeName=decor_ba_dyn_05_capsule, center=<0.344 0 0>, rotation=<rot 0 90 90>}]
 decor_ba_dyn_06 [{shapeType=Sphere, radius=0.2, shapeName=decor_ba_dyn_06_sphere}]
 decor_ct_dyn_01 [{shapeType=Capsule, length=2.686, radius=0.2, shapeName=decor_ct_dyn_01_capsule, center=<1.343 0 0>, rotation=<rot 0 -90 -90>}]
 decor_ct_dyn_02 [{shapeType=Capsule, length=1.894, radius=0.2, shapeName=decor_ct_dyn_02_capsule, center=<0.947 0 0>, rotation=<rot 0 -90 -90>}]
 decor_ct_dyn_03 [{shapeType=Capsule, length=1.893, radius=0.2, shapeName=decor_ct_dyn_03_capsule, center=<0.9465 0 0>, rotation=<rot 0 -90 -90>}]
 decor_ct_tassle [{shapeType=Sphere, radius=3.05, shapeName=decor_ct_tassle_sphere, center=<3.2 0 0>}]
 decor_dyn_all [{shapeType=Capsule, length=0.505, radius=0.505, shapeName=decor_dyn_all_capsule, center=<0.1651 0 0>, rotation=<rot 0 0 -90>}]
 decor_fr_dyn_01 [{shapeType=Capsule, length=1.021, radius=0.28, shapeName=decor_fr_dyn_01_capsule, center=<0.5105 0 0>, rotation=<rot 0 90 90>}]
 decor_fr_dyn_02 [{shapeType=Capsule, length=1.163, radius=0.28, shapeName=decor_fr_dyn_02_capsule, center=<0.5815 0 0>, rotation=<rot 0 90 90>}]
 decor_fr_dyn_03 [{shapeType=Capsule, length=1.94, radius=0.28, shapeName=decor_fr_dyn_03_capsule, center=<0.97 0 0>, rotation=<rot 0 90 90>}]
 decor_fr_dyn_04 [{shapeType=Capsule, length=1.653, radius=0.28, shapeName=decor_fr_dyn_04_capsule, center=<0.8265 0 0>, rotation=<rot 0 90 90>}]
 decor_fr_dyn_05 [{shapeType=Capsule, length=0.688, radius=0.28, shapeName=decor_fr_dyn_05_capsule, center=<0.344 0 0>, rotation=<rot 0 90 90>}]
 decor_fr_dyn_06 [{shapeType=Sphere, radius=0.2, shapeName=decor_fr_dyn_06_sphere}]
 dyn_decor_MSTR [{shapeType=Sphere, radius=0.2, shapeName=dyn_decor_MSTR_sphere}]
}
constraints [
 {bone1Name=decor_dyn_all, bone2Name=dyn_decor_MSTR, swing1LimitDegrees=0, swing2LimitDegrees=67.5, twistLimitDegrees=45, twistMotion=Locked}
 {
  bone1Name=decor_ct_dyn_01
  bone2Name=decor_dyn_all
  swing1LimitDegrees=24
  swing1Motion=Locked
  swing2LimitDegrees=24
  swing2Motion=Locked
  twistLimitDegrees=45
  twistMotion=Locked
 }
 {bone1Name=decor_ct_dyn_02, bone2Name=decor_ct_dyn_01, swing1LimitDegrees=45, swing2LimitDegrees=24, twistLimitDegrees=45, twistMotion=Locked}
 {bone1Name=decor_ct_dyn_03, bone2Name=decor_ct_dyn_02, swing1LimitDegrees=45, swing2LimitDegrees=24, twistLimitDegrees=45, twistMotion=Locked}
 {bone1Name=decor_ct_tassle, bone2Name=decor_ct_dyn_03, swing1LimitDegrees=45, swing2LimitDegrees=45, twistLimitDegrees=6.25}
 {bone1Name=decor_ba_dyn_01, bone2Name=decor_dyn_all, swing1LimitDegrees=20, swing2LimitDegrees=5, twistLimitDegrees=45, twistMotion=Locked}
 {bone1Name=decor_ba_dyn_02, bone2Name=decor_ba_dyn_01, swing1LimitDegrees=20, swing2LimitDegrees=5, twistLimitDegrees=45, twistMotion=Locked}
 {bone1Name=decor_ba_dyn_03, bone2Name=decor_ba_dyn_02, swing1LimitDegrees=20, swing2LimitDegrees=5, twistLimitDegrees=45, twistMotion=Locked}
 {bone1Name=decor_ba_dyn_04, bone2Name=decor_ba_dyn_03, swing1LimitDegrees=20, swing2LimitDegrees=5, twistLimitDegrees=45, twistMotion=Locked}
 {
  bone1Name=decor_ba_dyn_05
  bone2Name=decor_ba_dyn_04
  swing1LimitDegrees=45
  swing1Motion=Locked
  swing2LimitDegrees=45
  swing2Motion=Locked
  twistLimitDegrees=45
  twistMotion=Locked
 }
 {bone1Name=decor_fr_dyn_01, bone2Name=decor_dyn_all, swing1LimitDegrees=20, swing2LimitDegrees=5, twistLimitDegrees=45, twistMotion=Locked}
 {bone1Name=decor_fr_dyn_02, bone2Name=decor_fr_dyn_01, swing1LimitDegrees=20, swing2LimitDegrees=5, twistLimitDegrees=45, twistMotion=Locked}
 {bone1Name=decor_fr_dyn_03, bone2Name=decor_fr_dyn_02, swing1LimitDegrees=20, swing2LimitDegrees=5, twistLimitDegrees=45, twistMotion=Locked}
 {bone1Name=decor_fr_dyn_04, bone2Name=decor_fr_dyn_03, swing1LimitDegrees=20, swing2LimitDegrees=5, twistLimitDegrees=45, twistMotion=Locked}
 {
  bone1Name=decor_fr_dyn_05
  bone2Name=decor_fr_dyn_04
  swing1LimitDegrees=45
  swing1Motion=Locked
  swing2LimitDegrees=45
  swing2Motion=Locked
  twistLimitDegrees=45
  twistMotion=Locked
 }
 {
  bone1Name=decor_ba_dyn_06
  bone2Name=decor_ba_dyn_05
  swing1LimitDegrees=45
  swing1Motion=Locked
  swing2LimitDegrees=45
  swing2Motion=Locked
  twistLimitDegrees=45
  twistMotion=Locked
 }
 {
  bone1Name=decor_fr_dyn_06
  bone2Name=decor_fr_dyn_05
  swing1LimitDegrees=45
  swing1Motion=Locked
  swing2LimitDegrees=45
  swing2Motion=Locked
  twistLimitDegrees=45
  twistMotion=Locked
 }
 {
  bone1Name=decor_fr_dyn_06
  bone2Name=decor_ba_dyn_06
  swing1LimitDegrees=45
  swing1Motion=Locked
  swing2LimitDegrees=45
  swing2Motion=Locked
  twistLimitDegrees=45
  twistMotion=Locked
 }
 {bone1Name=decor_ba_dyn_06, bone2Name=decor_ct_tassle, swing1LimitDegrees=45, swing2LimitDegrees=45, twistLimitDegrees=6.25}
 {bone1Name=decor_fr_dyn_06, bone2Name=decor_ct_tassle, swing1LimitDegrees=45, swing2LimitDegrees=45, twistLimitDegrees=6.25}
]
